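# Masked unlit base material for critical-hit damage number text, drawing digits and a plus or minus sign.
# UV2 red blends digit and sign masks, diffuse, distance fields and normals, with a scalar offsetting the sign texture.
# Tint, UV2 green brightness, clamped distance field and diffuse are multiplied, then squared for contrast.
# DamageNumberWPO takes the blended normals and returns a pulse driving desaturation and a letter highlight multiplying the color.
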Material /Game/UI/DamageText/Default/DamageNumbers_CriticalHitBase
graph {
 properties {
  bEnableResponsiveAA=true
  blendMode=BLEND_Masked
  numCustomizedUVs=4
  pixelDepthOffsetMode=PDOM_Legacy
  shadingModel=MSM_Unlit
  translucencyLightingMode=TLM_Surface
  usedShadingModels=Unlit
 }
 nodes [
  Constant() {r=0.3}
  n1 = TextureCoordinate() {coordinateIndex=1}
  n2 = TextureSampleParameter2D(UVs=@n1) {group=Visuals, parameterName=Diffuse, samplerType=SAMPLERTYPE_LinearColor, texture=&/damage_text}
  n3 = Add(A=TextureCoordinate() {vTiling=0.5}, B=AppendVector(A=Constant(), B=ScalarParameter() {parameterName=+Or-}))
  n4 = TextureSampleParameter2D(UVs=@n3) {parameterName=+- Texture, texture=/Game/UI/DamageText/Default/+-Diffuse.+-Diffuse}
  n5 = TextureCoordinate() {coordinateIndex=2}
  n6 = ComponentMask(None=@n5) {r=true}
  LinearInterpolate(A=@n2.A, B=@n4.A, Alpha=@n6) {desc=o mask}
  n7 = VectorParameter() {group=Visuals, parameterName=Color, defaultValue=<rgba 0.5175 1.0366 1.5 0.5>}
  n8 = ComponentMask(None=@n5) {desc=Brightness value, g=true}
  n9 = Multiply(A=LinearInterpolate(
    A=TextureSample(UVs=@n1) {samplerType=SAMPLERTYPE_LinearGrayscale, texture=&/damage_text_distancefield}.R
    B=TextureSample(UVs=@n3) {samplerType=SAMPLERTYPE_LinearGrayscale, texture=&Default/+-Diffuse_DistanceField.+-Diffuse_DistanceField}.R
    Alpha=@n6
   )) {constB=1.125}
  n10 = Multiply(A=Multiply(A=Multiply(A=@n7.RGB, B=@n8), B=Clamp(None=@n9)), B=LinearInterpolate(A=@n2.R, B=@n4.R, Alpha=@n6)) {constB=0.25}
  n11 = TextureSampleParameter2D(UVs=@n1) {group=Visuals, parameterName=Normal, samplerType=SAMPLERTYPE_Normal, texture=&/damage_text_Normals}
  n12 = MaterialFunctionCall!(
   Normals=LinearInterpolate(
    A=@n11.RGB
    B=TextureSampleParameter2D(UVs=@n3) {parameterName=+- Normal, samplerType=SAMPLERTYPE_Normal, texture=&Default/+-Normal.+-Normal}.RGB
    Alpha=@n6
   )
   Debug texture coordinate input
   Animation
   Additonal Scale
   "Additional Camera Offset "
   "Additonal Screenspace XY offset "
  ) {materialFunction=/Game/UI/DamageText/Base/DamageNumberWPO}
  n13 = OneMinus(None=@n12.Vertex Shader - 0-1 highlight pulse)
  Multiply(A=Desaturation(None=Multiply(A=@n10, B=@n10), Fraction=Multiply(A=@n13) {constB=0.1}), B=@n12.Pixel shader Letter Highlight - uv 3)
 ]
}
