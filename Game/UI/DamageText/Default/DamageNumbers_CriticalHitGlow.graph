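# Translucent unlit material that draws the glow flare behind critical-hit damage numbers.
# DamageNumberWPO drives animation and flare scaling, with an added offset along the camera-to-vertex direction.
# Distance-field glyph or plus/minus samples, chosen by UV data, add to a rotated, sphere-masked halo, then a power curve sharpens the sum.
# That glow feeds depth-faded opacity and combines with a saturation-boosted, brightness-scaled color, squared for intensity.
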Material /Game/UI/DamageText/Default/DamageNumbers_CriticalHitGlow
graph {
 properties {
  bEnableResponsiveAA=true
  blendMode=BLEND_Translucent
  numCustomizedUVs=6
  pixelDepthOffsetMode=PDOM_Legacy
  shadingModel=MSM_Unlit
  translucencyLightingMode=TLM_Surface
  usedShadingModels=Unlit
 }
 nodes [
  n1 = ScalarParameter() {defaultValue=1.25, parameterName=Flare Size}
  n2 = Add(A=@n1)
  n3 = Add(A=Multiply(A=@n1) {constB=0.5})
  n4 = Subtract(A=ComponentMask(None=@n5.CustomUV3 - brightness values for the numbers) {r=true}) {constB=0.5}
  n5 = MaterialFunctionCall!(
   Normals
   Debug texture coordinate input
   Animation
   Additonal Scale=LinearInterpolate(B=AppendVector(A=@n2, B=@n3), Alpha=Divide(A=@n4)) {constA=1}
   "Additional Camera Offset "
   "Additonal Screenspace XY offset "
  ) {materialFunction=/Game/UI/DamageText/Base/DamageNumberWPO}
  n6 = Subtract(A=Add(A=@n5.WPO, B=WorldPosition().XYZ), B=CameraPositionWS())
  Add(A=@n5.WPO, B=Multiply(A=Normalize(VectorInput=@n6), B=ScalarParameter() {defaultValue=5, parameterName=Camera Offset}) {constB=5})
  n7 = Constant() {r=0.5}
  n8 = TextureSampleParameter2D(UVs=TextureCoordinate() {coordinateIndex=1}) {group=Visuals, parameterName=Diffuse, samplerType=SAMPLERTYPE_LinearGrayscale, texture=&/damage_text_distancefield}
  n9 = Constant()
  n10 = ScalarParameter() {parameterName=+Or-}
  n11 = TextureSampleParameter2D(UVs=Add(A=TextureCoordinate() {vTiling=0.5}, B=AppendVector(A=@n9, B=@n10))) {
   parameterName=+- Texture
   samplerType=SAMPLERTYPE_LinearGrayscale
   texture=/Game/UI/DamageText/Default/+-Diffuse_DistanceField.+-Diffuse_DistanceField
  }
  n12 = TextureCoordinate() {coordinateIndex=2}
  n13 = Multiply(A=LinearInterpolate(A=@n8.R, B=@n11.R, Alpha=ComponentMask(None=@n12) {r=true}), B=@n5.Pixel shader Letter Highlight - uv 3 0 to 1)
  n14 = MaterialFunctionCall(UVs=TextureCoordinate(), Texture Scale=Constant() {r=1.1}) {materialFunction=&ScaleUVsByCenter}
  n15 = MaterialFunctionCall(Float2=TextureCoordinate() {coordinateIndex=4}) {materialFunction=&BreakOutFloat2Components}
  n16 = Multiply(
   A=SphereMask(A=TextureCoordinate(), B=Constant() {r=0.5}) {attenuationRadius=0.5, hardnessPercent=0}
   B=TextureSample(UVs=MaterialFunctionCall(UVs=@n14.UVs, "Rotation Angle (0-1)"=@n15.R) {materialFunction=&CustomRotator}.Rotated Values) {samplerType=SAMPLERTYPE_LinearGrayscale, texture=&/damageNumberHalo}.R
  )
  n17 = Multiply(A=Power(Base=Add(A=@n13, B=@n16)) {constExponent=4}, B=ScalarParameter() {defaultValue=0.25, parameterName=Hit Glow Multiplier})
  Multiply(A=@n7, B=Clamp(None=DepthFade(Opacity=@n17) {fadeDistanceDefault=30}))
  n18 = Multiply(A=Desaturation(None=VectorParameter() {group=Visuals, parameterName=Color, defaultValue=<rgba 0.345 0.691 1 0.5>}.RGB, Fraction=Constant() {r=-0.4174})) {constB=3}
  n19 = ComponentMask(None=@n12) {desc=Brightness value, g=true}
  n20 = MaterialFunctionCall!(Base=@n17, Add=Multiply(A=Multiply(A=@n18, B=@n19), B=@n7)) {materialFunction=/Engine/Functions/Engine_MaterialFunctions02/Math/MultiplyAdd}
  Multiply(A=@n20.Result, B=@n20.Result)
 ]
}
