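# Default masked, lit material for damage number text, usable on particle sprites.
# UV2 red blends between the damage text atlas and a plus/minus texture, whose half is picked by a scalar V offset, for mask, diffuse and normals.
# A color parameter, UV2 green brightness and the letter highlight from DamageNumberWPO tint the glyphs, feeding scaled color and emissive branches.
# The blended alpha passes through a depth fade.
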Material /Game/UI/DamageText/Default/DamageNumbers_Mat
graph {
 properties {
  bEnableResponsiveAA=true
  bUsedWithParticleSprites=true
  blendMode=BLEND_Masked
  numCustomizedUVs=4
  pixelDepthOffsetMode=PDOM_Legacy
  previewMesh=&/standard_DamageNumbers8
  translucencyLightingMode=TLM_Surface
  usedShadingModels=Default Lit
 }
 nodes [
  Constant() {r=0.5}
  n1 = TextureCoordinate() {coordinateIndex=1}
  n2 = TextureSampleParameter2D(UVs=@n1) {group=Visuals, parameterName=Diffuse, samplerType=SAMPLERTYPE_LinearColor, texture=&/damage_text}
  n3 = Add(A=TextureCoordinate() {vTiling=0.5}, B=AppendVector(A=Constant(), B=ScalarParameter() {parameterName=+Or-}))
  n4 = TextureSampleParameter2D(UVs=@n3) {parameterName=+- Texture, texture=/Game/UI/DamageText/Default/+-Diffuse.+-Diffuse}
  n5 = TextureCoordinate() {coordinateIndex=2}
  n6 = ComponentMask(None=@n5) {r=true}
  DepthFade(Opacity=LinearInterpolate(A=@n2.A, B=@n4.A, Alpha=@n6) {desc=o mask}) {fadeDistanceDefault=30}
  n7 = VectorParameter() {group=Visuals, parameterName=Color, defaultValue=<rgba 0.9639 1 0.9798 0.5>}
  n8 = ComponentMask(None=@n5) {desc=Brightness value, g=true}
  n9 = TextureSampleParameter2D(UVs=@n1) {group=Visuals, parameterName=Normal, samplerType=SAMPLERTYPE_Normal, texture=&/damage_text_Normals}
  n10 = TextureSampleParameter2D(UVs=@n3) {parameterName=+- Normal, samplerType=SAMPLERTYPE_Normal, texture=&Default/+-Normal.+-Normal}
  n11 = Multiply(
   A=Multiply(A=Multiply(A=@n7.RGB, B=@n8), B=LinearInterpolate(A=@n2.R, B=@n4.R, Alpha=@n6)) {constB=0.25}
   B=MaterialFunctionCall!(
    Normals=LinearInterpolate(A=@n9.RGB, B=@n10.RGB, Alpha=@n6)
    Debug texture coordinate input
    Animation
    Additonal Scale
    "Additional Camera Offset "
    "Additonal Screenspace XY offset "
   ) {materialFunction=/Game/UI/DamageText/Base/DamageNumberWPO}.Pixel shader Letter Highlight - uv 3
  )
  MaterialFunctionCall(Emissive=Multiply(A=@n11) {constB=0.4, desc=Emissive}) {materialFunction=&MF_FX_EmissiveAdjustments}
  MaterialFunctionCall() {materialFunction=&TimeWithSpeedVariable}
  Multiply(A=@n11) {constB=0.5}
 ]
}
